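# Material function that samples animated turbulence and gust noise for a wind force shader, with a scalar input for the circular variant.
# Wind magnitude blends between minimum and maximum panning speeds for the supplied coordinates.
# Two samples of the engine wind turbulence texture at different scales, one set by the circular input, are averaged.
# The averaged RGB becomes a small centered random vector influence, and the averaged alpha is squared to modulate wind.
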MaterialFunction /Game/WaterAndWind/Materials/WindForceCylinder/AnimatedCylWindTextureSample
graph {nodes [
  n1 = FunctionInput() {inputName=Is Circular Shader, inputType=FunctionInput_Scalar}
  n2 = Constant() {desc=Min Wind Texture Panning Speed, r=0.06}
  n3 = Constant() {desc=Max Wind Texture Panning Speed, r=1.586}
  n4 = FunctionInput() {inputName=Wind Magnitude, inputType=FunctionInput_Scalar}
  n5 = Panner(
   Coordinate=FunctionInput() {inputName=Coordinates, inputType=FunctionInput_Vector2}
   Speed=Multiply(A=Constant2Vector() {g=0.25, r=-1}, B=LinearInterpolate(A=@n2, B=@n3, Alpha=@n4))
  ) {speedX=-0.25, speedY=0.1}
  n6 = TextureSample(UVs=Multiply(A=LinearInterpolate(Alpha=@n1) {constA=0.8778, constB=2}, B=@n5) {constA=0.8778}) {
   samplerType=SAMPLERTYPE_LinearColor
   texture=/Engine/Functions/Engine_MaterialFunctions02/ExampleContent/Textures/WindTurblenceVectorAndGustMagnitude
  }
  n7 = TextureSample(UVs=@n5) {
   samplerType=SAMPLERTYPE_LinearColor
   texture=&WindTurblenceVectorAndGustMagnitude
  }
  n8 = Multiply(A=Add(A=@n6.RGB, B=@n7.RGB)) {constB=0.5}
  FunctionOutput(None=Multiply(A=Subtract(A=@n8) {constB=0.5}) {constB=0.1, desc=Random Influence Amount}) {outputName=Random 2d Vector influence}
  n9 = Multiply(A=Add(A=@n6.A, B=@n7.A)) {constB=0.5}
  FunctionOutput(None=Multiply(A=@n9, B=@n9) {desc=Wind Variation}) {outputName=Modulate Wind}
 ]}
