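# Additive unlit storm wind material that outputs signed wind vectors as emissive color, with position and WPO from wind-vector particle helper functions.
# A radial cylinder falloff scales the wind direction, blended between a straight vector and a swirling tangent by "Circular Shader?", which also blends planar and radial texture coordinates.
# Static and animated magnitudes each get texture-driven modulation and random variation, then crossfade by distance from the animated center as the wave takes over.
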Material /Game/WaterAndWind/Materials/WindForceCylinder/StormWindShader
graph {
 properties {
  bAllowNegativeEmissiveColor=true
  bUseMaterialAttributes=true
  bUsedWithNanite=true
  blendMode=BLEND_Additive
  pixelDepthOffsetMode=PDOM_Legacy
  shadingModel=MSM_Unlit
  usedShadingModels=Unlit
 }
 nodes [
  n1 = ScalarParameter() {parameterName=Static Wind Magnitude}
  n2 = MaterialFunctionCall!(World position) {materialFunction=/Game/WaterAndWind/MaterialFunctions/Fortnite_WindVectorParticleHelperFunctions}
  n3 = VectorParameter() {parameterName=CylinderCenter, defaultValue=<rgba 0 0 0 1>}
  n4 = Clamp(None=ScalarParameter() {defaultValue=-1, group=OuterCylinder, parameterName=OverallOuterRadius}) {clampMode=CMODE_ClampMin}
  n5 = Subtract(
   A=Distance(
    A=ComponentMask(None=@n2.Pixel Shader Particle World Position - Adjusted For Offset) {g=true, r=true}
    B=ComponentMask(None=@n3.RGB) {g=true, r=true}
   )
   B=@n4
  )
  n6 = Constant() {r=0}
  n7 = Subtract(
   A=Clamp(None=ScalarParameter() {defaultValue=-1, group=OuterCylinder, parameterName=OverallInnerRadius}, Max=Subtract(A=@n4, B=@n6) {constB=0})
   B=@n4
  )
  n8 = VectorParameter() {parameterName=AnimatedCylinderCenter, defaultValue=<rgba -50 0 0 1>}
  n9 = Subtract(A=ComponentMask(None=WorldPosition().XYZ) {g=true, r=true}, B=ComponentMask(None=@n8.RGB) {g=true, r=true})
  n10 = MaterialFunctionCall(Vector=AppendVector(A=@n9, B=Constant())) {materialFunction=&SafeNormalize}
  n11 = Constant3Vector() {constant=<rgba 0 0 1 0>}
  n12 = Multiply(
   A=Clamp(None=Divide(A=@n5, B=@n7))
   B=LinearInterpolate(
    A=VectorParameter() {parameterName=WindVector, defaultValue=<rgba 1 0 0 1>}.RGB
    B=CrossProduct(A=@n10.Result, B=@n11)
    Alpha=ScalarParameter() {parameterName=Circular Shader?}
   )
  )
  n13 = ComponentMask(None=WorldPosition().XYZ) {g=true, r=true}
  n14 = DotProduct(A=ComponentMask(None=VectorParameter() {parameterName=WindCrossVector, defaultValue=<rgba 0 1 0 1>}.RGB) {g=true, r=true}, B=@n13)
  n15 = DotProduct(A=@n13, B=ComponentMask(None=VectorParameter() {parameterName=WindVector, defaultValue=<rgba 1 0 0 1>}.RGB) {g=true, r=true})
  n16 = MaterialFunctionCall(Vector or UVs=@n9) {materialFunction=&VectorToRadialValue}
  n17 = LinearInterpolate(
   A=Divide(A=AppendVector(A=@n14, B=@n15), B=ScalarParameter() {defaultValue=148.6203, parameterName=World unit tiling of panning textures})
   B=Divide(
    A=MaterialFunctionCall(XY=@n16.Radial Coordinates) {materialFunction=&Swizzle}.YX
    B=Constant2Vector() {g=0.04, r=-200}
   )
   Alpha=ScalarParameter() {parameterName=Circular Shader?}
  )
  n18 = ScalarParameter() {parameterName=Circular Shader?}
  n19 = MaterialFunctionCall!(Coordinates=@n17, Is Circular Shader=@n18, Wind Magnitude=@n1) {materialFunction=&/AnimatedCylWindTextureSample}
  n20 = MaterialFunctionCall!(Base=Multiply(A=Multiply(A=@n1, B=@n12), B=@n19.Modulate Wind), Add=@n19.Random 2d Vector influence) {materialFunction=/Engine/Functions/Engine_MaterialFunctions02/Math/MultiplyAdd}
  n21 = ScalarParameter() {parameterName=Animated Wind Magnitude}
  n22 = MaterialFunctionCall(Coordinates=@n17, Is Circular Shader=@n18, Wind Magnitude=@n21) {materialFunction=&AnimatedCylWindTextureSample}
  n23 = Multiply(
   A=Multiply(A=@n21, B=@n22.Modulate Wind)
   B=MaterialFunctionCall(Base=@n12, Add=@n22.Random 2d Vector influence) {materialFunction=&MultiplyAdd}.Result
  )
  n24 = Distance(
   A=ComponentMask(None=@n8.RGB) {g=true, r=true}
   B=ComponentMask(None=MaterialFunctionCall() {materialFunction=&Fortnite_WindVectorParticleHelperFunctions}.Pixel Shader Particle World Position - Adjusted For Offset) {g=true, r=true}
  )
  n25 = Clamp(None=ScalarParameter() {defaultValue=-1, group=Animated, parameterName=AnimatedOuterRadius}) {clampMode=CMODE_ClampMin, minDefault=0.01}
  n26 = Subtract(A=@n25, B=@n6) {constB=0}
  n27 = ScalarParameter() {defaultValue=-1, group=Animated, parameterName=AnimatedInnerRadius}
  n28 = Clamp(None=Divide(A=Subtract(A=@n24, B=@n25), B=Clamp(None=@n27, Max=@n26)))
  n29 = Multiply(
   A=LinearInterpolate(A=OneMinus(None=@n28), B=@n28, Alpha=ScalarParameter() {parameterName=Circular Shader?})
   B=Clamp(None=ScalarParameter() {parameterName=Wind Still 0 or Animating 1})
  ) {desc=Lerps to animated wind values as the wave takes over}
  MakeMaterialAttributes!(
   BaseColor
   Metallic
   Specular
   Roughness
   Anisotropy
   EmissiveColor=Multiply(A=LinearInterpolate(A=@n20.Result, B=@n23, Alpha=@n29) {desc=Normal}, B=ScalarParameter() {defaultValue=0.7, parameterName=Intensity Tweak})
   Opacity
   OpacityMask
   Normal
   Tangent
   WorldPositionOffset=MaterialFunctionCall() {materialFunction=&Fortnite_WindVectorParticleHelperFunctions}.Wind Particle WPO
   SubsurfaceColor
   ClearCoat
   ClearCoatRoughness
   AmbientOcclusion
   Refraction
   CustomizedUVs_0
   CustomizedUVs_1
   CustomizedUVs_2
   CustomizedUVs_3
   CustomizedUVs_4
   CustomizedUVs_5
   CustomizedUVs_6
   CustomizedUVs_7
   PixelDepthOffset
   ShadingModel
   Displacement
  )
 ]
}
